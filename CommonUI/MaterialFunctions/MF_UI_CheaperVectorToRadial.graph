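# UI material function that cheaply converts a Vector2 or UVs into radial coordinates, defaulting to texture coordinates.
# Subtracts the input from 0.5, then takes its length and a fast arctangent with red as Y and green as X.
# Scales the angle by 1/2π and wraps it with Frac to a 0-1 gradient, clamped at zero.
# Outputs angle and length, the swizzled pair, and the angle alone.
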MaterialFunction /CommonUI/MaterialFunctions/MF_UI_CheaperVectorToRadial
graph {nodes [
  n1 = FunctionInput(Preview=TextureCoordinate()) {
   bUsePreviewValueAsDefault=true
   description="Insert a vector 2 value to retrieve the vectors angle or insert uvs to retrieve radial coordinates. Note that (0,0) is not an acceptable number.  "
   inputName=Vector or UVs
   inputType=FunctionInput_Vector2
   previewValue=<1 1 0 1>
  }
  n2 = Subtract(B=@n1) {constA=0.5}
  n3 = Length(None=@n2) {is=class MaterialExpressionLength}
  n4 = Arctangent2Fast(Y=ComponentMask(None=@n2) {r=true}, X=ComponentMask(None=@n2) {g=true})
  n5 = Max(A=Frac(None=Multiply(A=@n4) {constB=0.1592, desc=1 / 2 Pi}), B=Constant() {r=0})
  FunctionOutput(None=AppendVector(A=@n5, B=@n3)) {
   description=The input coordinates or vector will be transformed into radial coordinates. In the case of a vector the angle will be output in one channel and the length of the vector in the other.
   outputName=Radial Coordinates
   sortPriority=1
  }
  FunctionOutput(None=AppendVector(A=@n3, B=@n5)) {description=Switched radial coordinates, outputName=Swizzled Radial Coordinates, sortPriority=2}
  FunctionOutput(None=@n5) {
   description=This will return the angle of of the input vector or a radial gradient in the case of input uvs.
   outputName=Vector Converted to Angle
   sortPriority=10
  }
 ]}
